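# Animated neon sign material showing a motorcycle art layer over a scrolling road, gradient sky with drifting stars and a gradient sun.
# UVs snap to a grid for a pixelated look, then feed Time-panned texture masks that layer the scene.
# A sine wobbles the motorcycle sideways, with a scrolling gradient-lookup glow attached.
# The color-washed result is multiplied by a scrolling mask and tiled LCD pattern, then fed to MF_NeonCityEmissiveBalance.
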Material /Game/Environments/Asteria/Sets/NeonCity/Techart/NeonSigns/SpecialAnimatedSigns/MotorCycle/M_MotorCycle
graph {
 properties {bUsedWithInstancedStaticMeshes=true, bUsedWithNanite=true, pixelDepthOffsetMode=PDOM_Legacy, usedShadingModels=Default Lit}
 nodes [
  Constant()
  MaterialFunctionCall() {materialFunction=&AppendMany}
  n1 = TextureCoordinate()
  n2 = ScalarParameter() {defaultValue=150, parameterName=Quantize Amount}
  n3 = Divide(A=Round(None=Multiply(A=@n1, B=@n2)), B=@n2)
  n4 = VectorParameter() {parameterName=asdfasd, defaultValue=<rgba 0 -0.276 0 1>}
  n5 = MaterialFunctionCall(UV Channel=Add(A=@n3, B=ComponentMask(None=@n4.RGB) {g=true, r=true})) {materialFunction=&LinearGradient}
  n6 = Multiply(A=@n3, B=AppendVector(A=ScalarParameter() {defaultValue=3, parameterName=RoadX}, B=ScalarParameter() {defaultValue=1, parameterName=RoadY}))
  n7 = Multiply(A=@n6, B=ScalarParameter() {defaultValue=1.25, parameterName=Scale})
  n8 = Multiply(A=@n7) {constB=2}
  n9 = AppendVector(
   A=Multiply(A=Time(), B=ScalarParameter() {defaultValue=0.1, parameterName=StarXo3}) {constB=0.1}
   B=ScalarParameter() {parameterName=StarYo3}
  )
  n10 = LinearInterpolate(
   A=LinearInterpolate(
    A=VectorParameter() {parameterName=BGA, defaultValue=<rgba 0 0.0068 0.0938 1>}.RGB
    B=VectorParameter() {parameterName=BGB, defaultValue=<rgba 0.3978 0 1 1>}.RGB
    Alpha=Power(Base=@n5.VGradient)
   )
   B=VectorParameter() {parameterName=Stars, defaultValue=<rgba 0.6831 0.5879 1 1>}.RGB
   Alpha=TextureSample(UVs=Add(A=@n8, B=@n9)) {texture=/Engine/EngineSky/T_Sky_Stars}.R
  )
  n11 = Constant2Vector() {g=0.5}
  n12 = LinearInterpolate(
   A=VectorParameter() {parameterName=SunA, defaultValue=<rgba 1 0.3341 0 1>}.RGB
   B=VectorParameter() {parameterName=SunB, defaultValue=<rgba 1 0 0.0134 1>}.RGB
   Alpha=MaterialFunctionCall(UV Channel=Add(A=@n3, B=@n11)) {materialFunction=&LinearGradient}.VGradient
  )
  n13 = SphereMask(A=@n6, B=ComponentMask(None=VectorParameter() {parameterName=SunPos, defaultValue=<rgba 0.9725 0 0 1>}.RGB) {g=true, r=true}) {attenuationRadius=0.55}
  n14 = Constant3Vector() {constant=<rgba 0.001 0.001 0.001 1>}
  n15 = Multiply(A=Time(), B=ScalarParameter() {defaultValue=0.2, parameterName=RoadXo3}) {constB=0.1}
  n16 = ScalarParameter() {parameterName=RoadYo3}
  n17 = TextureSample(UVs=Add(A=@n6, B=AppendVector(A=@n15, B=@n16))) {
   samplerType=SAMPLERTYPE_LinearColor
   texture=&/MotoSign_M
  }
  n18 = VectorParameter() {parameterName=Road, defaultValue=<rgba 0.1458 0.0183 0 1>}
  n19 = VectorParameter() {parameterName=RoadB, defaultValue=<rgba 0.02 0.0083 0.0151 1>}
  n20 = Multiply(A=Time(), B=ScalarParameter() {defaultValue=1.25, parameterName=RoadXo}) {constB=0.1}
  n21 = ScalarParameter() {defaultValue=-0.15, parameterName=RoadYo}
  n22 = TextureSample(UVs=Add(A=@n6, B=AppendVector(A=@n20, B=@n21))) {
   samplerType=SAMPLERTYPE_LinearColor
   texture=&MotoSign_M
  }
  n23 = Add(
   A=Power(Base=MaterialFunctionCall(UV Channel=@n3) {materialFunction=&LinearGradient}.VGradient)
   B=Multiply(A=@n22.B) {constB=0.15}
  )
  n24 = LinearInterpolate(
   A=LinearInterpolate(A=LinearInterpolate(A=@n10, B=@n12, Alpha=@n13), B=Multiply(A=@n14, B=@n17.G), Alpha=Round(None=@n17.G))
   B=LinearInterpolate(A=LinearInterpolate(A=@n18.RGB, B=@n19.RGB, Alpha=@n23), B=Constant3Vector() {constant=<rgba 0.375 0.1413 0.2901 1>}, Alpha=@n22.R)
   Alpha=@n22.A
  )
  n25 = ScalarParameter() {defaultValue=-1.7795, parameterName=MotA}
  n26 = Time()
  n27 = ScalarParameter() {defaultValue=0.1, parameterName=WobbleSpeed}
  n28 = LinearInterpolate(Alpha=Sine(None=Multiply(A=@n26, B=@n27))) {constA=0.8, constB=1.2}
  n29 = Add(A=@n7, B=AppendVector(A=Multiply(A=@n25, B=@n28), B=ScalarParameter() {defaultValue=-0.296, parameterName=MotB}))
  n30 = TextureSample(UVs=@n29) {
   samplerSource=SSM_Clamp_WorldGroupSettings
   texture=&/MotorGirl2_copy
  }
  n31 = ScalarParameter() {defaultValue=1.6983, parameterName=X}
  n32 = ScalarParameter() {defaultValue=-0.072, parameterName=Y}
  n33 = Multiply(A=Add(A=@n29, B=AppendVector(A=@n31, B=@n32)), B=Constant2Vector() {g=1, r=0.5}) {constB=1.5}
  n34 = TextureSample(UVs=@n33) {
   samplerSource=SSM_Clamp_WorldGroupSettings
   samplerType=SAMPLERTYPE_LinearColor
   texture=&/MotoSign_M2
  }
  n35 = Multiply(A=Time(), B=ScalarParameter() {defaultValue=2, parameterName=aaXo}) {constB=0.1}
  n36 = ScalarParameter() {defaultValue=-0.15, parameterName=aaYo}
  n37 = TextureSample(UVs=Add(A=@n33, B=AppendVector(A=@n35, B=@n36))) {
   samplerSource=SSM_Wrap_WorldGroupSettings
   samplerType=SAMPLERTYPE_LinearColor
   texture=&MotoSign_M2
  }
  n38 = Multiply(
   A=@n34.G
   B=TextureSample(UVs=@n37.B) {texture=/Game/Characters/Player/Male/Large/Bodies/M_LRG_Invader/Textures/T_InvaderGlow_Glass_Gradient}.RGB
  )
  n39 = LinearInterpolate(
   A=Add(A=LinearInterpolate(A=@n24, B=@n30.RGB, Alpha=@n30.RGB), B=Multiply(B=@n38) {constA=0.15})
   B=VectorParameter() {parameterName=Wash Color, defaultValue=<rgba 0.332 0 1 1>}.RGB
   Alpha=ScalarParameter() {defaultValue=0.02, parameterName=Wash Factor}
  )
  n40 = Multiply(A=TextureCoordinate()) {constB=250}
  MaterialFunctionCall!(Emissive=Multiply(
    A=Multiply(A=@n39, B=@n37.B)
    B=TextureSample(UVs=@n40) {samplerType=SAMPLERTYPE_Masks, texture=/Game/Environments/Asteria/Sets/NeonCity/Techart/Textures/T_LCD_Panels_LCDSoft}.RGB
   )) {materialFunction=/Game/Environments/Asteria/Sets/NeonCity/Techart/Screen/MaterialFunctions/MF_NeonCityEmissiveBalance}
  Multiply()
  Multiply()
  TextureCoordinate()
 ]
}
